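# Physics asset for the summer hat on the Banana character body, setting up bodies and constraints for its fringe, loops and bows.
# Spine and clavicle shapes form the body volume, with upper and lower fringe capsules constrained to spine_05 by tight angular limits.
# Left and right loop capsule chains start at spine_05 and join at their last two links with free-rotation constraints, closing the loop.
# Both bows attach to the right chain's final link.
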PhysicsAsset /BRCosmetics/Characters/Player/Male/Medium/Bodies/M_MED_Banana_Summer_Hat/Meshes/Parts/M_MED_Banana_Summer_Hat_Physics
bodies {
 clavicle_l [{shapeType=Box, extentX=1.01, extentY=1.01, extentZ=1.01, center=<18 0 0>, rotation=<rot 900 0 0>}]
 clavicle_r [{shapeType=Box, extentX=8.3466, extentY=58.4475, extentZ=59.9892, center=<-23.609 0.5522 -12.2838>, rotation=<rot 0 165.3679 -180>}]
 dyn_bow_l [{shapeType=Capsule, length=5.0824, radius=1.7533, center=<3.0346 -0.0835 -0.0616>, rotation=<rot 0 -98.6586 -84.3092>}]
 dyn_bow_r [{shapeType=Capsule, length=2.3908, radius=1.6186, center=<1.9493 0.0935 0.0315>, rotation=<rot 0 85.1611 -75.0486>}]
 dyn_hatFringeDown_b [{shapeType=Capsule, length=11.4095, radius=6.7764, center=<12.0348 2.1579 -0.0002>, rotation=<rot 90 0 0>}]
 dyn_hatFringeDown_bl [{shapeType=Capsule, length=11.4095, radius=6.7764, center=<13 -1.4085 -0.7134>, rotation=<rot 90 0 0>}]
 dyn_hatFringeDown_br [{shapeType=Capsule, length=11.4095, radius=6.7764, center=<14.8498 -0.0002 -1.1494>, rotation=<rot 90 0 0>}]
 dyn_hatFringeDown_f [{shapeType=Capsule, length=11.4095, radius=6.7764, center=<8.4272 -0.0007 -0.0002>, rotation=<rot 90 0 0>}]
 dyn_hatFringeDown_fl [{shapeType=Capsule, length=11.4095, radius=6.7764, center=<13 -0.0001 -1.0082>, rotation=<rot 90 0 0>}]
 dyn_hatFringeDown_fr [{shapeType=Capsule, length=11.4095, radius=6.7764, center=<13 0 -1.3998>, rotation=<rot 90 0 0>}]
 dyn_hatFringeDown_l [{shapeType=Capsule, length=11.4095, radius=6.7764, center=<13 0.0002 1.3911>, rotation=<rot 90 0 0>}]
 dyn_hatFringeDown_r [{shapeType=Capsule, length=11.4095, radius=6.7764, center=<11.6006 0 0.5908>, rotation=<rot 90 0 0>}]
 dyn_hatFringeUp_b [{shapeType=Capsule, length=7.1578, radius=6.5301, center=<3.5762 0 -0.0001>, rotation=<rot 90 0 0>}]
 dyn_hatFringeUp_f [{shapeType=Capsule, length=10.4413, radius=5.4033, center=<6.479 0 -0.0003>, rotation=<rot 90 0 0>}]
 dyn_hatFringeUp_l [{shapeType=Capsule, length=12.4442, radius=3.6942, center=<5.5497 0 -0.0001>, rotation=<rot 90 0 0>}]
 dyn_hatFringeUp_r [{shapeType=Capsule, length=11.4095, radius=6.7764, center=<9.911 0 -0.0002>, rotation=<rot 90 0 0>}]
 dyn_hatLoop02_r [{shapeType=Capsule, length=14.4008, radius=1.4399, center=<10.1276 0.0526 0.0029>, rotation=<rot 0 -90.3253 -89.934>}]
 dyn_hatLoop_01_Base_l [{shapeType=Box, extentX=1.01, extentY=1.01, extentZ=1.01, rotation=<rot 0 -90 0>}]
 dyn_hatLoop_01_Base_r [{shapeType=Box, extentX=1.01, extentY=1.01, extentZ=1.01, rotation=<rot 0 -90 0>}]
 dyn_hatLoop_01_l [{shapeType=Capsule, length=14.8025, radius=1.4883, center=<5.2035 -0.0572 0.0329>, rotation=<rot 0 91.2045 -89.9504>}]
 dyn_hatLoop_01_r [{shapeType=Capsule, length=14.5728, radius=1.5258, center=<4.6508 -0.0563 -0.1872>, rotation=<rot 0 90.8226 -88.4247>}]
 dyn_hatLoop_02_l [{shapeType=Capsule, length=14.4144, radius=1.4245, center=<10.1276 0.0362 -0.0714>, rotation=<rot 0 89.6616 -89.2616>}]
 dyn_hatLoop_03_l [{shapeType=Capsule, length=13.2747, radius=1.428, center=<7.9294 0.0215 -0.0512>, rotation=<rot 0 89.9663 -89.591>}]
 dyn_hatLoop_03_r [{shapeType=Capsule, length=13.2684, radius=1.4179, center=<7.9276 0.0226 0.0704>, rotation=<rot 0 -90.2444 -89.2895>}]
 dyn_hatLoop_04_l [{shapeType=Capsule, length=13.353, radius=1.4576, center=<7.9819 -0.07 -0.0952>, rotation=<rot 0 90.7721 -89.276>}]
 dyn_hatLoop_04_r [{shapeType=Capsule, length=13.3258, radius=1.4497, center=<7.9442 -0.1094 0.1062>, rotation=<rot 0 -89.0286 -89.2121>}]
 dyn_hatLoop_05_l [{shapeType=Capsule, length=13.3911, radius=1.5339, center=<8.175 0.0381 -0.0177>, rotation=<rot 0 87.9066 -87.3872>}]
 dyn_hatLoop_05_r [{shapeType=Capsule, length=9.8159, radius=1.4722, center=<5.6669 0.1146 0.1779>, rotation=<rot 0 -90.969 -88.1214>}]
 dyn_hatLoop_06_l [{shapeType=Capsule, length=10.6815, radius=1.2421, rotation=<rot 90 0 0>}]
 dyn_hatLoop_06_r [{shapeType=Capsule, length=6.1763, radius=1.4626, center=<4.0371 0.1091 0.126>, rotation=<rot 0 -91.9801 -87.8303>}]
 dyn_hatLoop_07_l [{shapeType=Capsule, length=4.5661, radius=2.0362, center=<-0.1547 -0.0205 0.0239>, rotation=<rot 90 0 0>}]
 dyn_hatLoop_07_r [{shapeType=Capsule, length=4.5799, radius=1.4436, center=<2.0373 0.1791 0.1166>, rotation=<rot 0 89.5926 -88.4556>}]
 dyn_hatLoop_08_l [{shapeType=Capsule, length=2.4959, radius=2.052, center=<0.1489 0.0275 -0.0364>, rotation=<rot 0 -79.2168 -76.6369>}]
 dyn_hatLoop_08_r [{shapeType=Capsule, length=2.0134, radius=2.1192, center=<1.7605 -0.0555 -0.0067>, rotation=<rot 0 -90.491 -86.4581>}]
 dyn_hatLoop_09_l [{shapeType=Capsule, length=1.7005, radius=1.2421, center=<0.7459 0 0>, rotation=<rot 90 0 0>}]
 dyn_hatLoop_09_r [{shapeType=Capsule, length=1.4394, radius=1.2421, rotation=<rot 90 0 0>}]
 spine_02 [{shapeType=Capsule, length=19.595, radius=13.7938, rotation=<rot 90 90 96.9239>}]
 spine_03 [{shapeType=Capsule, length=8.8132, radius=16.4114, rotation=<rot 90 0 0>}]
 spine_04 [{shapeType=Capsule, length=8.2598, radius=16.9648, rotation=<rot 90 0 0>}]
 spine_05 [
  {shapeType=Capsule, length=27.0449, radius=16.2279, rotation=<rot 90 0 0>}
  {shapeType=Box, extentX=91.6117, extentY=7.3387, extentZ=104.847, center=<0 -1.7655 0>}
  {shapeType=Box, extentX=103.7647, extentY=88.9358, extentZ=7.5, center=<-5.6283 -3.5521 26.348>, rotation=<rot 0 0 -31.1886>}
  {shapeType=Box, extentX=103.7647, extentY=88.9358, extentZ=7.5, center=<-5.6276 0.1401 -27.73>, rotation=<rot 0 0 41.75>}
 ]
}
constraints [
 {bone1Name=dyn_hatFringeUp_f, bone2Name=spine_05, swing1LimitDegrees=10, swing2LimitDegrees=10, twistLimitDegrees=6}
 {bone1Name=dyn_hatFringeUp_l, bone2Name=spine_05, swing1LimitDegrees=10, swing2LimitDegrees=10, twistLimitDegrees=6}
 {bone1Name=dyn_hatFringeUp_b, bone2Name=spine_05, swing1LimitDegrees=14, swing2LimitDegrees=10, twistLimitDegrees=6}
 {bone1Name=dyn_hatFringeUp_r, bone2Name=spine_05, swing1LimitDegrees=10, swing2LimitDegrees=10, twistLimitDegrees=6}
 {
  bone1Name=spine_05
  bone2Name=spine_04
  swing1LimitDegrees=45
  swing1Motion=Free
  swing2LimitDegrees=45
  swing2Motion=Free
  twistLimitDegrees=45
  twistMotion=Free
 }
 {
  bone1Name=spine_04
  bone2Name=spine_03
  swing1LimitDegrees=45
  swing1Motion=Free
  swing2LimitDegrees=45
  swing2Motion=Free
  twistLimitDegrees=45
  twistMotion=Free
 }
 {
  bone1Name=spine_03
  bone2Name=spine_02
  swing1LimitDegrees=45
  swing1Motion=Free
  swing2LimitDegrees=45
  swing2Motion=Free
  twistLimitDegrees=45
  twistMotion=Free
 }
 {bone1Name=dyn_hatFringeDown_l, bone2Name=spine_05, swing1LimitDegrees=25, swing2LimitDegrees=10, twistLimitDegrees=6}
 {bone1Name=dyn_hatFringeDown_fl, bone2Name=spine_05, swing1LimitDegrees=25, swing2LimitDegrees=10, twistLimitDegrees=6}
 {bone1Name=dyn_hatFringeDown_f, bone2Name=spine_05, swing1LimitDegrees=25, swing2LimitDegrees=10, twistLimitDegrees=6}
 {bone1Name=dyn_hatFringeDown_fr, bone2Name=spine_05, swing1LimitDegrees=25, swing2LimitDegrees=10, twistLimitDegrees=6}
 {bone1Name=dyn_hatFringeDown_r, bone2Name=spine_05, swing1LimitDegrees=25, swing2LimitDegrees=10, twistLimitDegrees=6}
 {bone1Name=dyn_hatFringeDown_br, bone2Name=spine_05, swing1LimitDegrees=25, swing2LimitDegrees=10, twistLimitDegrees=6}
 {bone1Name=dyn_hatFringeDown_b, bone2Name=spine_05, swing1LimitDegrees=25, swing2LimitDegrees=10, twistLimitDegrees=6}
 {bone1Name=dyn_hatFringeDown_bl, bone2Name=spine_05, swing1LimitDegrees=25, swing2LimitDegrees=10, twistLimitDegrees=6}
 {
  bone1Name=dyn_hatLoop_01_Base_r
  bone2Name=spine_05
  swing1LimitDegrees=45
  swing1Motion=Free
  swing2LimitDegrees=45
  swing2Motion=Free
  twistLimitDegrees=45
  twistMotion=Free
 }
 {
  bone1Name=dyn_hatLoop_01_Base_l
  bone2Name=spine_05
  swing1LimitDegrees=45
  swing1Motion=Free
  swing2LimitDegrees=45
  swing2Motion=Free
  twistLimitDegrees=45
  twistMotion=Free
 }
 {bone1Name=dyn_hatLoop_01_l, bone2Name=dyn_hatLoop_01_Base_l, swing1LimitDegrees=20, swing2LimitDegrees=8, twistLimitDegrees=8}
 {bone1Name=dyn_hatLoop_02_l, bone2Name=dyn_hatLoop_01_l, swing1LimitDegrees=20, swing2LimitDegrees=8, twistLimitDegrees=8}
 {bone1Name=dyn_hatLoop_03_l, bone2Name=dyn_hatLoop_02_l, swing1LimitDegrees=20, swing2LimitDegrees=8, twistLimitDegrees=8}
 {bone1Name=dyn_hatLoop_04_l, bone2Name=dyn_hatLoop_03_l, swing1LimitDegrees=20, swing2LimitDegrees=8, twistLimitDegrees=8}
 {bone1Name=dyn_hatLoop_05_l, bone2Name=dyn_hatLoop_04_l, swing1LimitDegrees=20, swing2LimitDegrees=8, twistLimitDegrees=8}
 {bone1Name=dyn_hatLoop_06_l, bone2Name=dyn_hatLoop_05_l, swing1LimitDegrees=20, swing2LimitDegrees=8, twistLimitDegrees=8}
 {bone1Name=dyn_hatLoop_07_l, bone2Name=dyn_hatLoop_06_l, swing1LimitDegrees=20, swing2LimitDegrees=8, twistLimitDegrees=8}
 {bone1Name=dyn_hatLoop_08_l, bone2Name=dyn_hatLoop_07_l, swing1LimitDegrees=20, swing2LimitDegrees=8, twistLimitDegrees=8}
 {bone1Name=dyn_hatLoop_09_l, bone2Name=dyn_hatLoop_08_l, swing1LimitDegrees=20, swing2LimitDegrees=8, twistLimitDegrees=8}
 {bone1Name=dyn_hatLoop_01_r, bone2Name=dyn_hatLoop_01_Base_r, swing1LimitDegrees=20, swing2LimitDegrees=8, twistLimitDegrees=8}
 {bone1Name=dyn_hatLoop02_r, bone2Name=dyn_hatLoop_01_r, swing1LimitDegrees=20, swing2LimitDegrees=8, twistLimitDegrees=8}
 {bone1Name=dyn_hatLoop_03_r, bone2Name=dyn_hatLoop02_r, swing1LimitDegrees=20, swing2LimitDegrees=8, twistLimitDegrees=8}
 {bone1Name=dyn_hatLoop_04_r, bone2Name=dyn_hatLoop_03_r, swing1LimitDegrees=20, swing2LimitDegrees=8, twistLimitDegrees=8}
 {bone1Name=dyn_hatLoop_05_r, bone2Name=dyn_hatLoop_04_r, swing1LimitDegrees=20, swing2LimitDegrees=8, twistLimitDegrees=8}
 {bone1Name=dyn_hatLoop_06_r, bone2Name=dyn_hatLoop_05_r, swing1LimitDegrees=20, swing2LimitDegrees=8, twistLimitDegrees=8}
 {bone1Name=dyn_hatLoop_07_r, bone2Name=dyn_hatLoop_06_r, swing1LimitDegrees=20, swing2LimitDegrees=8, twistLimitDegrees=8}
 {bone1Name=dyn_hatLoop_08_r, bone2Name=dyn_hatLoop_07_r, swing1LimitDegrees=20, swing2LimitDegrees=8, twistLimitDegrees=8}
 {bone1Name=dyn_hatLoop_09_r, bone2Name=dyn_hatLoop_08_r, swing1LimitDegrees=20, swing2LimitDegrees=8, twistLimitDegrees=8}
 {bone1Name=dyn_bow_r, bone2Name=dyn_hatLoop_09_r, swing1LimitDegrees=20, swing2LimitDegrees=12, twistLimitDegrees=12}
 {bone1Name=dyn_bow_l, bone2Name=dyn_hatLoop_09_r, swing1LimitDegrees=20, swing2LimitDegrees=12, twistLimitDegrees=12}
 {
  bone1Name=dyn_hatLoop_09_l
  bone2Name=dyn_hatLoop_09_r
  swing1LimitDegrees=20
  swing1Motion=Free
  swing2LimitDegrees=20
  swing2Motion=Free
  twistLimitDegrees=8
  twistMotion=Free
 }
 {
  bone1Name=clavicle_l
  bone2Name=spine_05
  swing1LimitDegrees=45
  swing1Motion=Free
  swing2LimitDegrees=45
  swing2Motion=Free
  twistLimitDegrees=45
  twistMotion=Free
 }
 {
  bone1Name=clavicle_r
  bone2Name=spine_05
  swing1LimitDegrees=45
  swing1Motion=Free
  swing2LimitDegrees=45
  swing2Motion=Free
  twistLimitDegrees=45
  twistMotion=Free
 }
 {
  bone1Name=dyn_hatLoop_08_l
  bone2Name=dyn_hatLoop_08_r
  swing1LimitDegrees=45
  swing1Motion=Free
  swing2LimitDegrees=45
  swing2Motion=Free
  twistLimitDegrees=45
  twistMotion=Free
 }
]
