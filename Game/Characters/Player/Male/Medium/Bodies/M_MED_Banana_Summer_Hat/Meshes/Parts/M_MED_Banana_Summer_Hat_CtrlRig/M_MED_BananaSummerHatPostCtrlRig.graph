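# Post-process Control Rig for the medium male Banana Summer Hat character part, adding procedural motion to the end of the hat's right strap loop.
# Begin Execution drives a Chain Harmonics Per Item unit rooted at dyn_hatLoop_06_r, with pendulum on and wave off.
# Its reach target follows the left loop's dyn_hatLoop_08_l bone, pulling the two loops toward each other.
# An older bone-name Chain Harmonics node with similar settings remains unconnected to execution; both have debug drawing on.
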ControlRigBlueprint /Game/Characters/Player/Male/Medium/Bodies/M_MED_Banana_Summer_Hat/Meshes/Parts/M_MED_Banana_Summer_Hat_CtrlRig/M_MED_BananaSummerHatPostCtrlRig
graphs [
 RigVMFunctionLibrary {links [], nodes []}
 RigVMModel {nodes [
   RigUnit_BeginExecution = &RigUnit_BeginExecution()
   /Script/ControlRig.RigUnit_ChainHarmonicsPerItem(
    ChainRoot=(Type=Bone,Name="dyn_hatLoop_06_r")
    DrawWorldOffset=(Rotation=(X=0,Y=0,Z=0,W=1),Translation=(X=0,Y=0,Z=0),Scale3D=(X=1,Y=1,Z=1))
    Pendulum=(bEnabled=True,PendulumStiffness=2,PendulumGravity=(X=0,Y=0,Z=0),PendulumBlend=0.75,PendulumDrag=0.98,PendulumMaximum=0.1,UnwindAxis=(X=0,Y=1,Z=0),UnwindMinimum=0.2,UnwindMaximum=0.05)
    Reach=(bEnabled=True,ReachTarget=(X=0,Y=0,Z=0),ReachAxis=(X=1,Y=0,Z=0),ReachMinimum=0.4,ReachEase=CubicEaseInOut)
    Speed=(X=1,Y=1,Z=1)
    Wave=(bEnabled=false,WaveFrequency=(X=1,Y=0.6,Z=0.8),WaveAmplitude=(X=0,Y=1,Z=0),WaveOffset=(X=0,Y=1,Z=2),WaveNoise=(X=0,Y=0,Z=0),WaveMaximum=1)
    WaveCurve=(EditorCurveData=(Keys=((),(Time=1,Value=1)),DefaultValue=340282346638528859811704183484516925440,PreInfinityExtrap=RCCE_Constant,PostInfinityExtrap=RCCE_Constant))
    bDrawDebug=True
    Reach.ReachTarget=&RigUnit_GetTransform(Item=(Type=Bone,Name="dyn_hatLoop_08_l")).Transform.Translation
    ExecutePin=@RigUnit_BeginExecution.ExecutePin
   ) {is=struct RigVMUnitNode}
   /Script/ControlRig.RigUnit_ChainHarmonics(
    ChainRoot=dyn_hatLoop_06_r
    DrawWorldOffset=(Rotation=(X=0,Y=0,Z=0,W=1),Translation=(X=0,Y=0,Z=0),Scale3D=(X=1,Y=1,Z=1))
    ExecutePin=()
    Pendulum=(bEnabled=True,PendulumStiffness=2,PendulumGravity=(X=0,Y=0,Z=0),PendulumBlend=0.75,PendulumDrag=0.98,PendulumMaximum=0.1,UnwindAxis=(X=0,Y=1,Z=0),UnwindMinimum=0.2,UnwindMaximum=0.05)
    Reach=(bEnabled=True,ReachTarget=(X=0,Y=0,Z=0),ReachAxis=(X=1,Y=0,Z=0),ReachMinimum=0.4,ReachMaximum=0,ReachEase=CubicEaseInOut)
    Speed=(X=1,Y=1,Z=1)
    Wave=(bEnabled=false,WaveFrequency=(X=1,Y=0.6,Z=0.8),WaveAmplitude=(X=0,Y=1,Z=0),WaveOffset=(X=0,Y=1,Z=2),WaveNoise=(X=0,Y=0,Z=0),WaveMaximum=1)
    WaveCurve=(EditorCurveData=(Keys=((),(Time=1,Value=1)),DefaultValue=340282346638528859811704183484516925440,PreInfinityExtrap=RCCE_Constant,PostInfinityExtrap=RCCE_Constant))
    bDrawDebug=True
    Reach.ReachTarget=&RigUnit_GetBoneTransform(Bone=dyn_hatLoop_08_l).Transform.Translation
   ) {is=struct RigVMUnitNode}
  ]}
]
hierarchy [
 root {loc=<0 0 0>, rot=<0 0 0 1>}
 pelvis {parent=root, loc=<0 -0.7071 93.1191>, rot=<0.0127 0.707 -0.0127 -0.707>}
 spine_01 {parent=pelvis, loc=<6.7494 -0.6388 0>, rot=<0 0 0.034 -0.9994>}
 spine_02 {parent=spine_01, loc=<7.2518 -0.035 0>, rot=<0 0 0.0238 -0.9997>}
 spine_03 {parent=spine_02, loc=<6.0104 0 0>, rot=<0 0 0.0149 -0.9999>}
 spine_04 {parent=spine_03, loc=<8.1024 0 0>, rot=<0 0 -0.0132 -0.9999>}
 spine_05 {parent=spine_04, loc=<10.0146 0 0>, rot=<0 0 -0.0015 -1>}
 clavicle_l {parent=spine_05, loc=<4.3575 -0.1466 -2.3223>, rot=<0.0465 -0.7289 -0.009 -0.6829>}
 upperarm_l {parent=clavicle_l, loc=<15.5111 0 0>, rot=<0.0226 -0.3234 -0.0088 -0.9459>}
 lowerarm_l {parent=upperarm_l, loc=<26.6514 0 0>, rot=<0.0391 -0.0099 -0.2453 -0.9686>}
 lowerarm_twist_02_l {parent=lowerarm_l, loc=<9.6199 0 0>, rot=<0 0 0 1>}
 upperarm_twist_01_l {parent=upperarm_l, loc=<0.0041 0 0>, rot=<0 0 0 1>}
 upperarm_twist_02_l {parent=upperarm_l, loc=<13.7941 0 0>, rot=<0 0 0 1>}
 clavicle_r {parent=spine_05, loc=<4.3575 -0.1466 2.3223>, rot=<0.7289 0.0465 -0.6829 0.009>}
 upperarm_r {parent=clavicle_r, loc=<-15.5111 0 0>, rot=<0.0226 -0.3234 -0.0088 -0.9459>}
 lowerarm_r {parent=upperarm_r, loc=<-26.6514 0 0>, rot=<0.0391 -0.0099 -0.2453 -0.9686>}
 lowerarm_twist_02_r {parent=lowerarm_r, loc=<-9.6199 0 0>, rot=<0 0 0 1>}
 upperarm_twist_01_r {parent=upperarm_r, loc=<-0.0041 0 0>, rot=<0 0 0 1>}
 upperarm_twist_02_r {parent=upperarm_r, loc=<-13.7942 0 0>, rot=<0 0 0 1>}
 neck_01 {parent=spine_05, loc=<11.3138 0.1529 0>, rot=<0 0 -0.2003 -0.9797>}
 neck_02 {parent=neck_01, loc=<4.5627 0 0>, rot=<0 0 0.0504 -0.9987>}
 head {parent=neck_02, loc=<5.3914 0 0>, rot=<0 0 0.115 -0.9934>}
 dyn_hatFringeUp_f {parent=spine_05, loc=<39.1415 2.8049 -0.0203>, rot=<-0.0851 0.0394 -0.4891 -0.8672>}
 dyn_hatFringeUp_l {parent=spine_05, loc=<39.1415 2.8049 -0.0203>, rot=<0.6465 -0.2834 -0.3354 -0.6238>}
 dyn_hatFringeUp_b {parent=spine_05, loc=<39.1415 2.8049 -0.0203>, rot=<0.9174 -0.3936 -0.0272 -0.0515>}
 dyn_hatFringeUp_r {parent=spine_05, loc=<39.1415 2.8049 -0.0203>, rot=<-0.6024 0.2727 -0.3631 -0.6564>}
 dyn_hatFringeDown_l {parent=spine_05, loc=<39.1415 2.8049 -0.0203>, rot=<-0.4982 0.5099 0.5392 0.4483>}
 dyn_hatFringeDown_fl {parent=spine_05, loc=<39.1415 2.8049 -0.0203>, rot=<0.2691 -0.2808 -0.7142 -0.582>}
 dyn_hatFringeDown_f {parent=spine_05, loc=<39.1415 2.8049 -0.0203>, rot=<-0.0171 0.018 -0.7783 -0.6274>}
 dyn_hatFringeDown_fr {parent=spine_05, loc=<39.1415 2.8049 -0.0203>, rot=<-0.2271 0.2392 -0.7346 -0.593>}
 dyn_hatFringeDown_r {parent=spine_05, loc=<39.1415 2.8049 -0.0203>, rot=<0.5024 -0.516 0.5343 0.4426>}
 dyn_hatFringeDown_br {parent=spine_05, loc=<39.1415 2.8049 -0.0203>, rot=<0.6503 -0.6598 0.2885 0.242>}
 dyn_hatFringeDown_b {parent=spine_05, loc=<39.1415 2.8049 -0.0203>, rot=<0.7009 -0.7107 0.0462 0.0387>}
 dyn_hatFringeDown_bl {parent=spine_05, loc=<39.1415 2.8049 -0.0203>, rot=<0.6356 -0.6429 -0.327 -0.2751>}
 dyn_hatLoop_01_Base_l {parent=spine_05, loc=<35.8876 4.3496 -17.3584>, rot=<0.0175 -0.0832 0.992 0.0932>}
 dyn_hatLoop_01_l {parent=dyn_hatLoop_01_Base_l, loc=<0 0 0>, rot=<0 0 0 1>}
 dyn_hatLoop_02_l {parent=dyn_hatLoop_01_l, loc=<10.588 0 0>, rot=<0.0221 0.0181 0.0039 -0.9996>}
 dyn_hatLoop_03_l {parent=dyn_hatLoop_02_l, loc=<15.2302 0 0>, rot=<-0.0218 0.0216 0.0005 -0.9995>}
 dyn_hatLoop_04_l {parent=dyn_hatLoop_03_l, loc=<11.8833 0 0>, rot=<0.0443 0.0357 0.0033 -0.9984>}
 dyn_hatLoop_05_l {parent=dyn_hatLoop_04_l, loc=<12.0141 0 0>, rot=<0.0958 0.0526 0.0126 -0.9939>}
 dyn_hatLoop_06_l {parent=dyn_hatLoop_05_l, loc=<10.1685 0 0>, rot=<0.1302 0.1 0.0803 -0.9832>}
 dyn_hatLoop_07_l {parent=dyn_hatLoop_06_l, loc=<5.3504 0 0>, rot=<0.1848 0.1065 0.1047 -0.9714>}
 dyn_hatLoop_08_l {parent=dyn_hatLoop_07_l, loc=<2.4207 0 0>, rot=<-0.0099 0.0544 0.0919 -0.9942>}
 dyn_hatLoop_09_l {parent=dyn_hatLoop_08_l, loc=<2.284 0 0>, rot=<0.0521 0.0207 0.1272 -0.9903>}
 dyn_hatLoop_01_Base_r {parent=spine_05, loc=<35.0856 4.7363 17.4933>, rot=<-0.0425 0.131 0.9868 0.0849>}
 dyn_hatLoop_01_r {parent=dyn_hatLoop_01_Base_r, loc=<0 0 0>, rot=<0 0 0 1>}
 dyn_hatLoop02_r {parent=dyn_hatLoop_01_r, loc=<9.785 0 0>, rot=<0.0349 0 0 -0.9994>}
 dyn_hatLoop_03_r {parent=dyn_hatLoop02_r, loc=<15.2112 0 0>, rot=<0.022 -0.002 0.0181 -0.9996>}
 dyn_hatLoop_04_r {parent=dyn_hatLoop_03_r, loc=<11.8916 0 0>, rot=<0.0246 -0.0326 -0.0086 -0.9991>}
 dyn_hatLoop_05_r {parent=dyn_hatLoop_04_r, loc=<11.9357 0 0>, rot=<-0.1542 -0.0682 0.0037 -0.9857>}
 dyn_hatLoop_06_r {parent=dyn_hatLoop_05_r, loc=<10.057 0 0>, rot=<-0.1266 -0.1029 0.0743 -0.9838>}
 dyn_hatLoop_07_r {parent=dyn_hatLoop_06_r, loc=<6.0267 0 0>, rot=<-0.118 -0.1353 0.1522 -0.9719>}
 dyn_hatLoop_08_r {parent=dyn_hatLoop_07_r, loc=<4.0782 0 0>, rot=<-0.0872 -0.0971 0.1805 -0.9749>}
 dyn_hatLoop_09_r {parent=dyn_hatLoop_08_r, loc=<1.8996 0 0>, rot=<0 0 0 1>}
 dyn_bow_l {parent=dyn_hatLoop_09_r, loc=<1.2534 0.5365 0.1939>, rot=<0.7502 0.1892 0.4217 0.4729>}
 dyn_bow_r {parent=dyn_hatLoop_09_r, loc=<-0.9126 0.4009 0.4983>, rot=<0.4975 0.4936 0.7015 0.1295>}
 thigh_l {parent=pelvis, loc=<-2.5693 -0.0109 -7.7501>, rot=<0.0012 0.0572 0.0155 -0.9982>}
 thigh_twist_01_l {parent=thigh_l, loc=<-21.6303 0 0>, rot=<0.0013 0 0 -1>}
 thigh_r {parent=pelvis, loc=<-2.5691 -0.0109 7.7501>, rot=<0.0572 -0.0012 0.9982 0.0155>}
 thigh_twist_01_r {parent=thigh_r, loc=<21.6303 0 0>, rot=<0.0013 0 0 -1>}
]
