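# Animation Blueprint for the Banana Summer Hat part of a medium male player cosmetic, following its attached parent mesh with secondary physics.
# The copied parent pose and curves feed a pre Control Rig, RigidBody simulation in spine_02 space with custom gravity, then a half-weight post Control Rig.
# Initialization caches body and head meshes, but the output chain uses neither; a layered bone blend sits unconnected.
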AnimBlueprint /BRCosmetics/Characters/Player/Male/Medium/Bodies/M_MED_Banana_Summer_Hat/Meshes/Parts/M_MED_Banana_Summer_Hat_AnimBP
target_skeleton=/Game/Characters/Player/Male/Medium/Bodies/M_MED_Banana_Summer_Hat/Meshes/Parts/M_MED_Banana_Summer_Hat_Skeleton
graphs [
 AnimGraph {nodes [
   Layered blend per bone(BlendPoses_0=Copy Pose From Mesh(SourceMeshComponent=Get Body_Mesh())) {layerSetup [{branchFilters [
       {boneName=upperarm_r}
       {boneName=upperarm_l}
       {boneName=clavicle_l}
       {boneName=clavicle_r}
       {boneName=None}
       {boneName=None}
       {boneName=None}
       {boneName=None}
       {boneName=spine_02}
       {boneName=spine_03}
       {boneName=spine_04}
       {boneName=spine_05}
      ]}]}
   n1 = Control Rig(Source=Copy Pose From Mesh() {bCopyCurves=true, bUseAttachedParent=true}) {
    alphaBoolBlend {blendOption=Linear}
    controlRigAssetReference {blueprintRigClass=/Game/Characters/Player/Male/Medium/Bodies/M_MED_Banana_Summer_Hat/Meshes/Parts/M_MED_Banana_Summer_Hat_CtrlRig/M_MED_BananaSummerHat_PreCtrlRig.M_MED_BananaSummerHat_PreCtrlRig_C}
   }
   n2 = RigidBody(ComponentPose=Local To Component(LocalPose=@n1)) {
    bClampLinearTranslationLimitToRefPose=true
    bOverrideWorldGravity=true
    overridePhysicsAsset=&/M_MED_Banana_Summer_Hat_Physics
    simulationSpace=BaseBoneSpace
    alphaBoolBlend {blendOption=Linear}
    baseBoneRef {boneName=spine_02}
    componentAppliedLinearAccClamp=<4000 4000 4000>
    componentLinearAccScale {x=0.3}
    overrideWorldGravity {z=-500}
   }
   n3 = Control Rig(Source=Component To Local(ComponentPose=@n2), Alpha=0.5) {
    alphaBoolBlend {blendOption=Linear}
    controlRigAssetReference {blueprintRigClass=/Game/Characters/Player/Male/Medium/Bodies/M_MED_Banana_Summer_Hat/Meshes/Parts/M_MED_Banana_Summer_Hat_CtrlRig/M_MED_BananaSummerHatPostCtrlRig.M_MED_BananaSummerHatPostCtrlRig_C}
   }
   Output Pose(Result=@n3) {layerGroup=None}
  ]}
 EventGraph {nodes [
   Event BlueprintInitializeAnimation()
   > Set Body_Mesh(Body_Mesh=GetOwnerSkeletalMeshComponent())
   > n1 = Cast To PlayerPawn_Athena(Object=TryGetPawnOwner())
   > n2 = Set Pawn Owner(Pawn Owner:Player Pawn Athena Object Reference=@n1) -> (Output_Get:Player Pawn Athena Object Reference)
   > Set Head_Mesh(Head_Mesh:Skeletal Mesh Component Object Reference=GetSkeletalMeshForPartType(self=@n2, PartType=Head)) -> (Output_Get:Skeletal Mesh Component Object Reference)
   Event BlueprintUpdateAnimation()
  ]}
]
variables [
 Body_Mesh {type=SkeletalMeshComponent}
 Head_Mesh {type=SkeletalMeshComponent}
 Pawn Owner {type=PlayerPawn_Athena_C}
]
